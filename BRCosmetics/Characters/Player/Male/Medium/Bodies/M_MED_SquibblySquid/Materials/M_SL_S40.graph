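# Character material for Super Levels (SL) skin styles, wrapping recolor and effect layers between the shared Character Parent 2023 PreFX and PostFX functions.
# Ice and IceSprite functions process the PreFX result into a Material reroute.
# Channel-selectable masks drive layered lerps of color, metallic, specular, roughness and emissive, with AO and GMap influence.
# Static switches add glacier tinting, a Fresnel thin-film metal with curve-atlas reflectance, and FN hair.
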Material /BRCosmetics/Characters/Player/Male/Medium/Bodies/M_MED_SquibblySquid/Materials/M_SL_S40
graph {
 properties {
  bAutomaticallySetUsageInEditor=false
  bUseMaterialAttributes=true
  bUsedWithMeshDeformer=true
  bUsedWithMorphTargets=true
  bUsedWithSkeletalMesh=true
  pixelDepthOffsetMode=PDOM_Legacy
  shadingModel=MSM_FromMaterialExpression
  usedShadingModels=Default Lit | Subsurface | Clear Coat | Subsurface Profile | Cloth
 }
 nodes [
  n1 = NamedRerouteUsage()
  n2 = MaterialFunctionCall() {materialFunction=&MF_CharacterParent_2023_PreFX}
  n3 = MaterialFunctionCall!(In=@n1, Mask Input=@n2.M) {materialFunction=/Game/Global/Materials/Character_Parent_2023/Hair_Parent_2023/MaterialFunctions/MF_FN_Hair_2023}
  MaterialFunctionCall(
   Material=StaticSwitchParameter(True=BlendMaterialAttributes(A=@n1, B=@n3.Result, Alpha=@n3.Result), False=@n1) {group=SL 01 Global, parameterName=UseFNHair2_0}
   M=@n2.M
   Pre Skinned Local Position=@n2.Pre skinned Local Position
   Distance Blend=@n2.Pre skinned Local Position
   Fresnel=@n2.Pre skinned Local Position
  ) {materialFunction=&MF_CharacterParent_2023_PostFX}
  n4 = NamedRerouteUsage()
  n5 = NamedRerouteUsage()
  n6 = NamedRerouteUsage()
  NamedRerouteDeclaration(None=MaterialFunctionCall!(
    Material=MaterialFunctionCall!(Material=@n2.Result, Accent Mask=@n4, Mask Input=@n2.M, GlacierIceMap=@n5, TechArtMask=@n6) {materialFunction=&/MF_NewIce_SL}.Material
    MakeUPMask=NamedRerouteUsage()
   ) {materialFunction=/BRCosmetics/Characters/Player/Female/Medium/Bodies/F_MED_IcicleCane_Pitch/Skins/Bright/Material/MF_IceSprite_SL}.Result) {name=Material, nodeColor=<rgba 1 0 0.8672 1>}
  NamedRerouteDeclaration(None=VectorParameter() {group=SL 01 Global, parameterName=SL.TertiaryColor2, sortPriority=2}.RGB) {name=TertiaryColor_2, nodeColor=<rgba 0.2248 0.2274 1 1>}
  n7 = GetMaterialAttributes(None=NamedRerouteUsage()) {attributeGetTypes [
    #1
    #2
    #3
    #4
    #5
    #6
   ]}
  NamedRerouteDeclaration(None=@n7.Roughness) {name=Roughness, nodeColor=<rgba 0.7109 1 0 1>}
  n8 = VectorParameter() {group=SL 01 Global, parameterName=SL.GlacierColor, sortPriority=100, defaultValue=<rgba 1 1 1 1>}
  n9 = VectorParameter() {group=SL 01 Global, parameterName=SL.SecondaryColor, sortPriority=1}
  n10 = TextureSampleParameter2D() {
   group=SL 01 Global
   parameterName=SL.GlacierIceMap
   samplerSource=SSM_Wrap_WorldGroupSettings
   samplerType=SAMPLERTYPE_Masks
   sortPriority=100
   texture=/BRCosmetics/Characters/Player/Male/Medium/Bodies/M_MED_BriskReign_Glacier/Textures/T_BriskReign_Glacier_Body_FX
  }
  n11 = NamedRerouteDeclaration(None=Power(Base=@n10.G, Exp=ScalarParameter() {group=SL 01 Global, parameterName=SL.GlacierIceMap_Influence, sortPriority=100}) {constExponent=10}) {name=GlacierIceMap, nodeColor=<rgba 1 0 0.6328 1>}
  NamedRerouteDeclaration(None=StaticSwitchParameter(True=LinearInterpolate(A=@n8.RGB, B=@n9.RGB, Alpha=@n11) {constA=1}, False=@n9.RGB) {group=SL 01 Global, parameterName=SL.UseGlacierIceMap, sortPriority=99}) {name=SecondaryColor, nodeColor=<rgba 0.9688 0 1 1>}
  NamedRerouteDeclaration(None=ScalarParameter() {group=SL 01 Global, parameterName=SL.Tertiary_Metallic, sortPriority=2}) {name=TertiaryMetallic, nodeColor=<rgba 0.2248 0.2274 1 1>}
  NamedRerouteDeclaration(None=ScalarParameter() {group=SL 01 Global, parameterName=SL.Tertiary_Specular, sortPriority=2}) {name=TertiarySpecular, nodeColor=<rgba 0.2248 0.2274 1 1>}
  NamedRerouteDeclaration(None=ScalarParameter() {group=SL 01 Global, parameterName=SL.Tertiary_Roughness, sortPriority=2}) {name=TertiaryRoughness, nodeColor=<rgba 0.2248 0.2274 1 1>}
  NamedRerouteDeclaration(None=ScalarParameter() {group=SL 01 Global, parameterName=SL.Accent_Metallic, sortPriority=3}) {name=AccentMetallic, nodeColor=<rgba 0.1854 0.7843 1 1>}
  NamedRerouteDeclaration(None=ScalarParameter() {group=SL 01 Global, parameterName=SL.Accent_Specular, sortPriority=3}) {name=AccentSpecular, nodeColor=<rgba 0.1854 0.7843 1 1>}
  NamedRerouteDeclaration(None=ScalarParameter() {group=SL 01 Global, parameterName=SL.Accent_Roughness, sortPriority=3}) {name=AccentRoughness, nodeColor=<rgba 0.1854 0.7843 1 1>}
  NamedRerouteDeclaration(None=VectorParameter() {group=SL 01 Global, parameterName=SL.TertiaryColor, sortPriority=2}.RGB) {name=TertiaryColor, nodeColor=<rgba 0.2248 0.2274 1 1>}
  NamedRerouteDeclaration(None=VectorParameter() {group=SL 01 Global, parameterName=SL.AccentColor, sortPriority=3}.RGB) {name=AccentColor, nodeColor=<rgba 0.1854 0.7843 1 1>}
  NamedRerouteDeclaration(None=SmoothStep(
    Min=ScalarParameter() {defaultValue=0.5, group=SL 01 Global, parameterName=SL.TA_Mask_SmoothStepMin, sortPriority=100}
    Value=@n2.Tech Art Mask
   ) {constMin=0.5}) {name=TechArtMask, nodeColor=<rgba 1 0.4922 0 1>}
  NamedRerouteDeclaration(None=ScalarParameter() {group=SL 01 Global, parameterName=SL.Secondary_Metallic, sortPriority=1}) {name=SecondaryMetallic, nodeColor=<rgba 0.5938 0 1 1>}
  NamedRerouteDeclaration(None=ScalarParameter() {group=SL 01 Global, parameterName=SL.Secondary_Specular, sortPriority=1}) {name=SecondarySpecular, nodeColor=<rgba 0.5938 0 1 1>}
  NamedRerouteDeclaration(None=ScalarParameter() {group=SL 01 Global, parameterName=SL.Secondary_Roughness, sortPriority=1}) {name=SecondaryRoughness, nodeColor=<rgba 0.5938 0 1 1>}
  n12 = NamedRerouteUsage()
  n13 = Multiply(A=VectorParameter() {group=SL 01 Global, parameterName=SL.ExtraDetailTint, sortPriority=2, defaultValue=<rgba 1 1 1 1>}.RGB, B=@n12)
  n14 = Add(A=NamedRerouteUsage(), B=NamedRerouteUsage())
  n15 = LinearInterpolate(
   A=LinearInterpolate(A=LinearInterpolate(A=@n13, B=@n12, Alpha=@n14), B=NamedRerouteUsage(), Alpha=NamedRerouteUsage())
   B=NamedRerouteUsage()
   Alpha=NamedRerouteUsage()
  )
  n16 = NamedRerouteUsage()
  n17 = NamedRerouteUsage()
  NamedRerouteDeclaration(None=LinearInterpolate(A=LinearInterpolate(A=@n15, B=@n16, Alpha=@n17), B=NamedRerouteUsage(), Alpha=NamedRerouteUsage())) {name=SL_BaseColor, nodeColor=<rgba 0 0.0469 1 1>}
  n18 = LinearInterpolate(
   A=LinearInterpolate(A=NamedRerouteUsage(), B=NamedRerouteUsage(), Alpha=NamedRerouteUsage())
   B=NamedRerouteUsage()
   Alpha=NamedRerouteUsage()
  )
  n19 = NamedRerouteUsage()
  n20 = NamedRerouteUsage()
  NamedRerouteDeclaration(None=LinearInterpolate(A=LinearInterpolate(A=@n18, B=@n19, Alpha=@n20), B=NamedRerouteUsage(), Alpha=NamedRerouteUsage())) {name=SL_Metallic, nodeColor=<rgba 0.5106 0.2321 1 1>}
  n21 = LinearInterpolate(
   A=Add(A=NamedRerouteUsage(), B=ScalarParameter() {group=SL 01 Global, parameterName=SL.Base_Roughness_Bias, sortPriority=1})
   B=NamedRerouteUsage()
   Alpha=NamedRerouteUsage()
  )
  n22 = NamedRerouteUsage()
  n23 = NamedRerouteUsage()
  n24 = LinearInterpolate(
   A=LinearInterpolate(A=LinearInterpolate(A=@n21, B=@n22, Alpha=@n23), B=NamedRerouteUsage(), Alpha=NamedRerouteUsage())
   B=NamedRerouteUsage()
   Alpha=NamedRerouteUsage()
  )
  n25 = NamedRerouteUsage()
  NamedRerouteDeclaration(None=LinearInterpolate(A=@n24, B=Multiply(A=@n24, B=@n25), Alpha=ScalarParameter() {group=SL 01 Global, parameterName=SL.Roughness_GMap_Influence})) {name=SL_Roughness, nodeColor=<rgba 1 0.338 0 1>}
  n26 = LinearInterpolate(
   A=LinearInterpolate(A=NamedRerouteUsage(), B=NamedRerouteUsage(), Alpha=NamedRerouteUsage())
   B=NamedRerouteUsage()
   Alpha=NamedRerouteUsage()
  )
  n27 = NamedRerouteUsage()
  n28 = NamedRerouteUsage()
  NamedRerouteDeclaration(None=LinearInterpolate(A=LinearInterpolate(A=@n26, B=@n27, Alpha=@n28), B=NamedRerouteUsage(), Alpha=NamedRerouteUsage())) {name=SL_Specular, nodeColor=<rgba 0.9918 1 0.1851 1>}
  n29 = MaterialFunctionCall(Float3=@n2.M) {desc="R = AO\r\nG = GMap\r\nB = Skin", materialFunction=&BreakOutFloat3Components}
  NamedRerouteDeclaration(None=@n29.B) {name=SkinMask, nodeColor=<rgba 0.2187 1 0 1>}
  n30 = NamedRerouteUsage()
  NamedRerouteDeclaration(None=ChannelMaskParameter(None=@n30) {group=SL 01 Global, parameterName=SL.PrimaryMaskChannel, sortPriority=-32}.R) {name=PrimaryMask, nodeColor=<rgba 0 0.8951 1 1>}
  n31 = TextureSampleParameter2D() {
   group=SL 01 Global
   parameterName=SL.FX2Mask
   samplerSource=SSM_Wrap_WorldGroupSettings
   samplerType=SAMPLERTYPE_Masks
   sortPriority=-30
   texture=&T_BlackMask
  }
  n32 = StaticSwitchParameter(True=@n31.RGB, False=NamedRerouteUsage()) {defaultValue=true, parameterName=SL.UseFX2Mask}
  NamedRerouteDeclaration(None=ChannelMaskParameter(None=@n32) {group=SL 01 Global, parameterName=SL.EmissiveDetailMaskChannel, sortPriority=-27}.R) {name=EmissiveDetailMask, nodeColor=<rgba 1 0.6843 0.0075 1>}
  n33 = ChannelMaskParameter(None=@n30) {group=SL 01 Global, maskChannel=Blue, parameterName=SL.SecondaryMaskChannel, sortPriority=-29, defaultValue=<rgba 0 0 1 0>}
  NamedRerouteDeclaration(None=@n33.R) {name=SecondaryMask, nodeColor=<rgba 0.1719 0 1 1>}
  n34 = ChannelMaskParameter(None=@n32) {group=SL 01 Global, maskChannel=Blue, parameterName=SL.AccentMaskChannel_02, sortPriority=-27, defaultValue=<rgba 0 0 1 0>}
  NamedRerouteDeclaration(None=@n34.R) {name=AccentMask_02, nodeColor=<rgba 1 0 0.4688 1>}
  NamedRerouteDeclaration(None=VectorParameter() {group=SL 01 Global, parameterName=SL.AccentColor_02, sortPriority=3}.RGB) {name=AccentColor_02, nodeColor=<rgba 0.1854 0.7843 1 1>}
  NamedRerouteDeclaration(None=ScalarParameter() {group=SL 01 Global, parameterName=SL.Accent_Metallic_02, sortPriority=3}) {name=AccentMetallic_02, nodeColor=<rgba 0.1854 0.7843 1 1>}
  NamedRerouteDeclaration(None=ScalarParameter() {group=SL 01 Global, parameterName=SL.Accent_Specular_02, sortPriority=3}) {name=AccentSpecular_02, nodeColor=<rgba 0.1854 0.7843 1 1>}
  NamedRerouteDeclaration(None=VectorParameter() {group=SL 01 Global, parameterName=SL.EmissiveDetailColor, sortPriority=3}.RGB) {name=EmissiveDetailColor, nodeColor=<rgba 0.1854 0.7843 1 1>}
  NamedRerouteDeclaration(None=@n7.Emissive Color) {name=Emissive, nodeColor=<rgba 1 0.4435 0.2797 1>}
  NamedRerouteDeclaration(None=@n7.Normal) {name=Normal, nodeColor=<rgba 1 0.1712 0.7455 1>}
  NamedRerouteDeclaration(None=ScalarParameter() {group=SL 01 Global, parameterName=SL.Accent_Roughness_02, sortPriority=3}) {name=AccentRoughness_02, nodeColor=<rgba 0.1854 0.7843 1 1>}
  NamedRerouteDeclaration(None=StaticBoolParameter() {parameterName=UseTertiaryGlitter}) {name=UseTertiaryGlitter, nodeColor=<rgba 0 1 0.1657 1>}
  n35 = NamedRerouteUsage()
  n36 = ScalarParameter() {defaultValue=1, group=SL 02 Metal Effect, parameterName=SL.Secondary_ThinFilmScale, sortPriority=5}
  n37 = Fresnel() {exponent=1}
  n38 = Power(Base=@n37, Exp=ScalarParameter() {defaultValue=0.5, group=SL 02 Metal Effect, parameterName=SL.Secondary_ThinFilmWarp, sortPriority=4})
  n39 = TextureSampleParameter2D(UVs=Add(A=Multiply(A=@n36, B=@n38), B=ScalarParameter() {group=SL 02 Metal Effect, parameterName=SL.Secondary_ThinFilmOffset})) {
   group=SL 02 Metal Effect
   parameterName=SL.Secondary Thin Film Texture
   samplerSource=SSM_Wrap_WorldGroupSettings
   sortPriority=1
   texture=/Game/Global/Textures/Default/Thinfilm/T_ThinFilm_Spectrum_COLOR
  }
  n40 = ScalarParameter() {defaultValue=0.5, group=SL 02 Metal Effect, parameterName=SL.Global_ThinFilmIntensity, sortPriority=2}
  n41 = Power(Base=@n37, Exp=ScalarParameter() {defaultValue=2, group=SL 02 Metal Effect, parameterName=SL.Global_ThinFilmExponent, sortPriority=3})
  n42 = Multiply(A=LinearInterpolate(Alpha=NamedRerouteUsage()) {constA=1, constB=0.01}, B=@n41)
  n43 = ScalarParameter() {group=SL 02 Metal Effect, parameterName=SL.Global_RoughnessInfluence}
  n44 = LinearInterpolate(A=@n35, B=Add(A=@n35, B=Multiply(A=@n39.RGB, B=@n40)), Alpha=Saturate(None=LinearInterpolate(A=@n41, B=@n42, Alpha=@n43)))
  n45 = CameraVectorWS()
  n46 = PixelNormalWS()
  n47 = CurveAtlasRowParameter(CurveTime=Clamp(None=DotProduct(A=@n45, B=@n46))) {
   atlas=/Game/Global/Textures/Default/Curve/FN_Character_Atlas
   curve=/Game/Characters/Player/Female/Medium/Bodies/F_MED_Ms_Alpine/Textures/AluminiumreflectanceCurve
   group=SL 02 Metal Effect
   parameterName=SL.Metal LUT Curve
  }
  n48 = Multiply(A=@n44, B=ChannelMaskParameter(None=@n47) {group=SL 02 Metal Effect, parameterName=SL.LUTChannel}.R)
  n49 = Saturate(None=ScalarParameter() {defaultValue=1, group=SL 02 Metal Effect, parameterName=SL.MetalLutIntensity})
  NamedRerouteDeclaration(None=LinearInterpolate(A=@n35, B=LinearInterpolate(A=@n44, B=@n48, Alpha=@n49), Alpha=NamedRerouteUsage())) {name=SL_BaseColor_Metal, nodeColor=<rgba 1 0 0.0937 1>}
  n50 = ChannelMaskParameter(None=@n30) {group=SL 01 Global, maskChannel=Green, parameterName=SL.TertiaryMaskChannel, sortPriority=-28, defaultValue=<rgba 0 1 0 0>}
  NamedRerouteDeclaration(None=@n50.R) {name=TertiaryMask, nodeColor=<rgba 0.1484 1 0 1>}
  NamedRerouteDeclaration(None=ScalarParameter() {group=SL 01 Global, parameterName=SL.EmissiveDetailStrength, sortPriority=3}) {name=EmissiveDetailStrength, nodeColor=<rgba 0.1854 0.7843 1 1>}
  n51 = LinearInterpolate(A=NamedRerouteUsage(), B=NamedRerouteUsage(), Alpha=NamedRerouteUsage())
  n52 = NamedRerouteUsage()
  NamedRerouteDeclaration(None=LinearInterpolate(A=NamedRerouteUsage(), B=Multiply(A=@n51, B=@n52), Alpha=NamedRerouteUsage())) {name=SL_Emissive, nodeColor=<rgba 0 0.7422 1 1>}
  NamedRerouteDeclaration(None=MaterialFunctionCall(
    Brightness=ScalarParameter() {defaultValue=1, group=SL 01 Global, parameterName=SL.EmissiveFresnel_Brightness, sliderMin=4, sortPriority=6}
    Exponent=ScalarParameter() {defaultValue=1.5, group=SL 01 Global, parameterName=SL.EmissiveFresnel_Exponent, sortPriority=7}
    Vertex / Pixel Normal=PixelNormalWS()
   ) {materialFunction=&MF_Fresnel_FX}.Fresnel Inverted) {name=EmissiveFresnel, nodeColor=<rgba 1 0 0.2812 1>}
  NamedRerouteDeclaration(None=VectorParameter() {group=SL 01 Global, parameterName=SL.EmissiveFresnelColor, sortPriority=3}.RGB) {name=EmissiveFresnelColor, nodeColor=<rgba 0.1854 0.7843 1 1>}
  NamedRerouteDeclaration(None=VectorParameter() {group=SL 01 Global, parameterName=SL.EmissiveDetailBaseColor, sortPriority=3}.RGB) {name=EmissiveDetaiBaseColor, nodeColor=<rgba 0.1854 0.7843 1 1>}
  NamedRerouteDeclaration(None=@n29.R) {name=AO, nodeColor=<rgba 1 0.1172 0 1>}
  NamedRerouteDeclaration(None=@n29.G) {name=GMap, nodeColor=<rgba 0 0.7309 1 1>}
  NamedRerouteDeclaration(None=@n31.B) {name=IcicleCane_MakeUPMask, nodeColor=<rgba 0.4531 0 1 1>}
  n53 = ChannelMaskParameter(None=@n32) {group=SL 01 Global, maskChannel=Green, parameterName=SL.AccentMaskChannel, sortPriority=-27, defaultValue=<rgba 0 1 0 0>}
  NamedRerouteDeclaration(None=@n53.R) {name=AccentMask, nodeColor=<rgba 1 0 0.4687 1>}
  n54 = LinearInterpolate(
   A=StaticSwitchParameter(True=NamedRerouteUsage(), False=NamedRerouteUsage()) {group=SL 02 Metal Effect, parameterName=SL.UseSLMetal, sortPriority=-30}
   B=NamedRerouteUsage()
   Alpha=NamedRerouteUsage()
  )
  n55 = NamedRerouteUsage()
  n56 = SetMaterialAttributes!(
   MaterialAttributes=NamedRerouteUsage()
   Base Color=LinearInterpolate(A=@n54, B=Multiply(A=@n54, B=@n55), Alpha=ScalarParameter() {group=SL 01 Global, parameterName=SL.BaseColor_AO_Influence})
   Metallic=NamedRerouteUsage()
   Specular=NamedRerouteUsage()
   Roughness=NamedRerouteUsage()
   Emissive Color=NamedRerouteUsage()
  ) {attributeSetTypes [
    #1
    #2
    #3
    #4
    #5
   ]}
  NamedRerouteDeclaration(None=@n56) {name=SuperLevelsMaterial, nodeColor=<rgba 0 0.3906 1 1>}
  NamedRerouteDeclaration(None=@n7.Base Color) {name=BaseColor, nodeColor=<rgba 0 1 0.25 1>}
  NamedRerouteDeclaration(None=@n7.Metallic) {name=Metallic, nodeColor=<rgba 1 0 0.0703 1>}
  NamedRerouteDeclaration(None=@n7.Specular) {name=Specular, nodeColor=<rgba 0 0.3906 1 1>}
  NamedRerouteUsage()
  n57 = TextureCoordinate()
  n58 = LinearInterpolate(
   A=ScalarParameter() {defaultValue=3, group=Diamond_Base, parameterName=NormalTile}
   B=ScalarParameter() {defaultValue=3, group=Diamond_Base, parameterName=Secondary_NormalTile}
   Alpha=NamedRerouteUsage()
  )
  SetMaterialAttributes!(
   MaterialAttributes=@n2.Result
   Base Color=TextureSample(UVs=Multiply(A=@n57, B=@n58)) {texture=/UVEditor/Textures/UVEditorColorGrid_Color}.RGB
  ) {attributeSetTypes [#1]}
 ]
}
