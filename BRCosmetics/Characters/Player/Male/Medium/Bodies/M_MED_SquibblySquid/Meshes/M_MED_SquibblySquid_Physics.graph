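# Physics asset for the male medium SquibblySquid body cosmetic, with capsule bodies on upper spine, clavicles, chest, left extra lower arm, poncho and tentacle bones.
# Free joints link spine, clavicles and tiny poncho anchor capsules, feeding six three-segment flap chains (front, back and side per shoulder).
# Flap joints have tight swing limits and mostly zero twist, so each bends mainly on one axis.
# Tentacle, extra-arm and chest bodies have no constraints.
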PhysicsAsset /BRCosmetics/Characters/Player/Male/Medium/Bodies/M_MED_SquibblySquid/Meshes/M_MED_SquibblySquid_Physics
bodies {
 chest_arms_ft_jnt [
  {shapeType=Capsule, length=24, radius=10, shapeName=chest_arms_ft_jnt_capsule, center=<-5 0 -5>, rotation=<rot 0 0 90>}
  {shapeType=Capsule, length=8, radius=5, shapeName=chest_arms_ft_jnt_capsule, center=<3.6183 -0.9573 -2.5944>, rotation=<rot 0.0001 0 60>}
 ]
 clavicle_l [{shapeType=Capsule, length=6, radius=8, shapeName=clavicle_l_capsule, center=<5 0 -1>, rotation=<rot 0 90 90>}]
 clavicle_r [{shapeType=Capsule, length=6, radius=8, shapeName=clavicle_r_capsule, center=<-5 0 1>, rotation=<rot 0 90 90>}]
 extra_lowerarm_l_01_jnt [{shapeType=Capsule, length=10, radius=5, shapeName=extra_lowerarm_l_01_jnt_capsule, center=<5 1 0>, rotation=<rot 0 -90 90>}]
 extra_lowerarm_l_02_jnt [{shapeType=Capsule, length=10, radius=5, shapeName=extra_lowerarm_l_02_jnt_capsule, center=<5 1 0>, rotation=<rot 0 90 90>}]
 extra_lowerarm_l_03_jnt [{shapeType=Capsule, length=10, radius=5, shapeName=extra_lowerarm_l_03_jnt_capsule, center=<5 1 0>, rotation=<rot 0 90 90>}]
 extra_lowerarm_l_04_jnt [{shapeType=Capsule, length=10, radius=5, shapeName=extra_lowerarm_l_04_jnt_capsule, center=<5 1 0>, rotation=<rot 0 90 90>}]
 poncho_clavicle_l [{shapeType=Capsule, length=0.505, radius=0.505, shapeName=poncho_clavicle_l_capsule, rotation=<rot 0 -90 0>}]
 poncho_clavicle_r [{shapeType=Capsule, length=0.505, radius=0.505, shapeName=poncho_clavicle_r_capsule, rotation=<rot 0 -90 0>}]
 poncho_flap_01_bk_l [{shapeType=Capsule, length=5, radius=4.5, shapeName=poncho_flap_01_bk_l_capsule, center=<5 0 2>, rotation=<rot 0 90 90>}]
 poncho_flap_01_bk_r [{shapeType=Capsule, length=5, radius=4.5, shapeName=poncho_flap_01_bk_r_capsule, center=<5 0 2>, rotation=<rot 0 90 90>}]
 poncho_flap_01_ft_l [{shapeType=Capsule, length=5, radius=3, shapeName=poncho_flap_01_ft_l_capsule, center=<5 0 2>, rotation=<rot 0 90 90>}]
 poncho_flap_01_ft_r [{shapeType=Capsule, length=5, radius=3, shapeName=poncho_flap_01_ft_r_capsule, center=<5 0 2>, rotation=<rot 0 90 90>}]
 poncho_flap_01_l [{shapeType=Capsule, length=5, radius=3, shapeName=poncho_flap_01_l_capsule, center=<5 0 2>, rotation=<rot 0 90 90>}]
 poncho_flap_01_r [{shapeType=Capsule, length=5, radius=3, shapeName=poncho_flap_01_r_capsule, center=<5 0 2>, rotation=<rot 0 90 90>}]
 poncho_flap_02_bk_l [{shapeType=Capsule, length=15, radius=5, shapeName=poncho_flap_02_bk_l_capsule, center=<5 0 2>, rotation=<rot 0 0 90>}]
 poncho_flap_02_bk_r [{shapeType=Capsule, length=15, radius=5, shapeName=poncho_flap_02_bk_r_capsule, center=<5 0 2>, rotation=<rot 0 0 90>}]
 poncho_flap_02_ft_l [{shapeType=Capsule, length=15, radius=5, shapeName=poncho_flap_02_ft_l_capsule, center=<5 -5 2>, rotation=<rot 0 0 90>}]
 poncho_flap_02_ft_r [{shapeType=Capsule, length=20, radius=5, shapeName=poncho_flap_02_ft_r_capsule, center=<5 5 2>, rotation=<rot 0 0 90>}]
 poncho_flap_02_l [{shapeType=Capsule, length=12, radius=5, shapeName=poncho_flap_02_l_capsule, center=<5 0 4>, rotation=<rot 0 0 90>}]
 poncho_flap_02_r [{shapeType=Capsule, length=12, radius=5, shapeName=poncho_flap_02_r_capsule, center=<5 0 4>, rotation=<rot 0 0 90>}]
 poncho_flap_03_bk_l [{shapeType=Capsule, length=15, radius=5, shapeName=poncho_flap_03_bk_l_capsule, center=<5 0 2>, rotation=<rot 0 0 90>}]
 poncho_flap_03_bk_r [{shapeType=Capsule, length=15, radius=5, shapeName=poncho_flap_03_bk_r_capsule, center=<5 0 2>, rotation=<rot 0 0 90>}]
 poncho_flap_03_ft_l [{shapeType=Capsule, length=15, radius=5, shapeName=poncho_flap_03_ft_l_capsule, center=<5 -5 0>, rotation=<rot 0.0001 -10 89.9999>}]
 poncho_flap_03_ft_r [{shapeType=Capsule, length=20, radius=5, shapeName=poncho_flap_03_ft_r_capsule, center=<5 5 0>, rotation=<rot 0.0001 10 80.0001>}]
 poncho_flap_03_l [{shapeType=Capsule, length=12, radius=5, shapeName=poncho_flap_03_l_capsule, center=<8 0 4>, rotation=<rot 0 0 90>}]
 poncho_flap_03_r [{shapeType=Capsule, length=12, radius=5, shapeName=poncho_flap_03_r_capsule, center=<8 0 4>, rotation=<rot 0 0 90>}]
 poncho_flap_MSTR_bk_l [{shapeType=Capsule, length=0.505, radius=0.505, shapeName=poncho_flap_MSTR_bk_l_capsule, rotation=<rot 0 -90 0>}]
 poncho_flap_MSTR_bk_r [{shapeType=Capsule, length=0.505, radius=0.505, shapeName=poncho_flap_MSTR_bk_r_capsule, rotation=<rot 0 -90 0>}]
 poncho_flap_MSTR_ft_l [{shapeType=Capsule, length=0.505, radius=0.505, shapeName=poncho_flap_MSTR_ft_l_capsule, rotation=<rot 0 -90 0>}]
 poncho_flap_MSTR_ft_r [{shapeType=Capsule, length=0.505, radius=0.505, shapeName=poncho_flap_MSTR_ft_r_capsule, rotation=<rot 0 -90 0>}]
 poncho_flap_MSTR_l [{shapeType=Capsule, length=0.505, radius=0.505, shapeName=poncho_flap_MSTR_l_capsule, rotation=<rot 0 -90 0>}]
 poncho_flap_MSTR_r [{shapeType=Capsule, length=0.505, radius=0.505, shapeName=poncho_flap_MSTR_r_capsule, rotation=<rot 0 -90 0>}]
 spine_03 [{shapeType=Capsule, length=12, radius=10, shapeName=spine_03_capsule}]
 spine_04 [{shapeType=Capsule, length=12, radius=11, shapeName=spine_04_capsule}]
 spine_05 [{shapeType=Capsule, length=12, radius=11, shapeName=spine_05_capsule}]
 tentacle_01_l [{shapeType=Capsule, length=6, radius=5, shapeName=tentacle_01_l_capsule, center=<5 0 -2>, rotation=<rot 0 90 90>}]
 tentacle_01_r [{shapeType=Capsule, length=6, radius=5, shapeName=tentacle_01_r_capsule, center=<-5 0 2>, rotation=<rot 0 90 90>}]
 tentacle_02_l [{shapeType=Capsule, length=6, radius=5, shapeName=tentacle_02_l_capsule, center=<5 0 -1>, rotation=<rot 0 90 90>}]
 tentacle_02_r [{shapeType=Capsule, length=6, radius=5, shapeName=tentacle_02_r_capsule, center=<-5 0 1>, rotation=<rot 0 90 90>}]
 tentacle_03_l [{shapeType=Capsule, length=6, radius=5, shapeName=tentacle_03_l_capsule, center=<5 0 -0.5>, rotation=<rot 0 90 90>}]
 tentacle_03_r [{shapeType=Capsule, length=6, radius=5, shapeName=tentacle_03_r_capsule, center=<-5 0 0.5>, rotation=<rot 0 90 90>}]
 tentacle_04_l [{shapeType=Capsule, length=6, radius=5, shapeName=tentacle_04_l_capsule, center=<5 0 -0.5>, rotation=<rot 0 90 90>}]
 tentacle_04_r [{shapeType=Capsule, length=6, radius=5, shapeName=tentacle_04_r_capsule, center=<-5 0 0.5>, rotation=<rot 0 90 90>}]
 tentacle_05_l [{shapeType=Capsule, length=12, radius=5, shapeName=tentacle_05_l_capsule, center=<10 0 -0.5>, rotation=<rot 0 90 90>}]
 tentacle_05_r [{shapeType=Capsule, length=12, radius=5, shapeName=tentacle_05_r_capsule, center=<-10 0 0.5>, rotation=<rot 0 90 90>}]
 tentacle_06_l [{shapeType=Capsule, length=6, radius=4.5, shapeName=tentacle_06_l_capsule, center=<4 0 -0.5>, rotation=<rot 0 90 90>}]
 tentacle_06_r [{shapeType=Capsule, length=6, radius=4.5, shapeName=tentacle_06_r_capsule, center=<-4 0 0.5>, rotation=<rot 0 90 90>}]
 tentacle_07_l [{shapeType=Capsule, length=6, radius=4.5, shapeName=tentacle_07_l_capsule, center=<4 0 -0.5>, rotation=<rot 0 90 90>}]
 tentacle_07_r [{shapeType=Capsule, length=6, radius=4.5, shapeName=tentacle_07_r_capsule, center=<-4 0 0.5>, rotation=<rot 0 90 90>}]
 tentacle_08_l [{shapeType=Capsule, length=6, radius=4.5, shapeName=tentacle_08_l_capsule, center=<4 0 -0.5>, rotation=<rot 0 90 90>}]
 tentacle_08_r [{shapeType=Capsule, length=6, radius=4.5, shapeName=tentacle_08_r_capsule, center=<-4 0 0.5>, rotation=<rot 0 90 90>}]
 tentacle_09_l [{shapeType=Capsule, length=6, radius=4.5, shapeName=tentacle_09_l_capsule, center=<4 0 -0.5>, rotation=<rot 0 90 90>}]
 tentacle_09_r [{shapeType=Capsule, length=6, radius=4.5, shapeName=tentacle_09_r_capsule, center=<-4 0 0.5>, rotation=<rot 0 90 90>}]
 tentacle_10_l [{shapeType=Capsule, length=12, radius=4.5, shapeName=tentacle_10_l_capsule, center=<8 0 -0.5>, rotation=<rot 0 90 90>}]
 tentacle_10_r [{shapeType=Capsule, length=12, radius=4.5, shapeName=tentacle_10_r_capsule, center=<-8 0 0.5>, rotation=<rot 0 90 90>}]
}
constraints [
 {
  bone1Name=spine_04
  bone2Name=spine_03
  swing1LimitDegrees=45
  swing1Motion=Free
  swing2LimitDegrees=45
  swing2Motion=Free
  twistLimitDegrees=45
  twistMotion=Free
 }
 {
  bone1Name=spine_05
  bone2Name=spine_04
  swing1LimitDegrees=45
  swing1Motion=Free
  swing2LimitDegrees=45
  swing2Motion=Free
  twistLimitDegrees=45
  twistMotion=Free
 }
 {
  bone1Name=clavicle_l
  bone2Name=spine_05
  swing1LimitDegrees=45
  swing1Motion=Free
  swing2LimitDegrees=45
  swing2Motion=Free
  twistLimitDegrees=45
  twistMotion=Free
 }
 {
  bone1Name=clavicle_r
  bone2Name=spine_05
  swing1LimitDegrees=45
  swing1Motion=Free
  swing2LimitDegrees=45
  swing2Motion=Free
  twistLimitDegrees=45
  twistMotion=Free
 }
 {
  bone1Name=poncho_flap_MSTR_bk_l
  bone2Name=poncho_clavicle_l
  swing1LimitDegrees=45
  swing1Motion=Free
  swing2LimitDegrees=45
  swing2Motion=Free
  twistLimitDegrees=45
  twistMotion=Free
 }
 {bone1Name=poncho_flap_01_bk_l, bone2Name=poncho_flap_MSTR_bk_l, swing1LimitDegrees=3, swing2LimitDegrees=15, twistLimitDegrees=0}
 {bone1Name=poncho_flap_02_bk_l, bone2Name=poncho_flap_01_bk_l, swing1LimitDegrees=3, swing2LimitDegrees=15, twistLimitDegrees=0}
 {bone1Name=poncho_flap_03_bk_l, bone2Name=poncho_flap_02_bk_l, swing1LimitDegrees=3, swing2LimitDegrees=15, twistLimitDegrees=0}
 {
  bone1Name=poncho_clavicle_l
  bone2Name=clavicle_l
  swing1LimitDegrees=45
  swing1Motion=Free
  swing2LimitDegrees=45
  swing2Motion=Free
  twistLimitDegrees=45
  twistMotion=Free
 }
 {
  bone1Name=poncho_flap_MSTR_ft_l
  bone2Name=poncho_clavicle_l
  swing1LimitDegrees=45
  swing1Motion=Free
  swing2LimitDegrees=45
  swing2Motion=Free
  twistLimitDegrees=45
  twistMotion=Free
 }
 {bone1Name=poncho_flap_01_ft_l, bone2Name=poncho_flap_MSTR_ft_l, swing1LimitDegrees=3, swing2LimitDegrees=15, twistLimitDegrees=0}
 {bone1Name=poncho_flap_02_ft_l, bone2Name=poncho_flap_01_ft_l, swing1LimitDegrees=3, swing2LimitDegrees=15, twistLimitDegrees=0}
 {bone1Name=poncho_flap_03_ft_l, bone2Name=poncho_flap_02_ft_l, swing1LimitDegrees=3, swing2LimitDegrees=15, twistLimitDegrees=0}
 {
  bone1Name=poncho_flap_MSTR_l
  bone2Name=poncho_clavicle_l
  swing1LimitDegrees=45
  swing1Motion=Free
  swing2LimitDegrees=45
  swing2Motion=Free
  twistLimitDegrees=45
  twistMotion=Free
 }
 {bone1Name=poncho_flap_01_l, bone2Name=poncho_flap_MSTR_l, swing1LimitDegrees=3, swing2LimitDegrees=15, twistLimitDegrees=0}
 {bone1Name=poncho_flap_02_l, bone2Name=poncho_flap_01_l, swing1LimitDegrees=0, swing2LimitDegrees=15, twistLimitDegrees=0}
 {bone1Name=poncho_flap_03_l, bone2Name=poncho_flap_02_l, swing1LimitDegrees=0, swing2LimitDegrees=15, twistLimitDegrees=0}
 {
  bone1Name=poncho_clavicle_r
  bone2Name=clavicle_r
  swing1LimitDegrees=45
  swing1Motion=Free
  swing2LimitDegrees=45
  swing2Motion=Free
  twistLimitDegrees=45
  twistMotion=Free
 }
 {
  bone1Name=poncho_flap_MSTR_bk_r
  bone2Name=poncho_clavicle_r
  swing1LimitDegrees=45
  swing1Motion=Free
  swing2LimitDegrees=45
  swing2Motion=Free
  twistLimitDegrees=45
  twistMotion=Free
 }
 {bone1Name=poncho_flap_01_bk_r, bone2Name=poncho_flap_MSTR_bk_r, swing1LimitDegrees=3, swing2LimitDegrees=15, twistLimitDegrees=0}
 {bone1Name=poncho_flap_02_bk_r, bone2Name=poncho_flap_01_bk_r, swing1LimitDegrees=3, swing2LimitDegrees=15, twistLimitDegrees=0}
 {bone1Name=poncho_flap_03_bk_r, bone2Name=poncho_flap_02_bk_r, swing1LimitDegrees=3, swing2LimitDegrees=15, twistLimitDegrees=0}
 {
  bone1Name=poncho_flap_MSTR_ft_r
  bone2Name=poncho_clavicle_r
  swing1LimitDegrees=45
  swing1Motion=Free
  swing2LimitDegrees=45
  swing2Motion=Free
  twistLimitDegrees=45
  twistMotion=Free
 }
 {bone1Name=poncho_flap_01_ft_r, bone2Name=poncho_flap_MSTR_ft_r, swing1LimitDegrees=3, swing2LimitDegrees=15, twistLimitDegrees=0}
 {bone1Name=poncho_flap_02_ft_r, bone2Name=poncho_flap_01_ft_r, swing1LimitDegrees=3, swing2LimitDegrees=15, twistLimitDegrees=10}
 {bone1Name=poncho_flap_03_ft_r, bone2Name=poncho_flap_02_ft_r, swing1LimitDegrees=3, swing2LimitDegrees=15, twistLimitDegrees=10}
 {
  bone1Name=poncho_flap_MSTR_r
  bone2Name=poncho_clavicle_r
  swing1LimitDegrees=45
  swing1Motion=Free
  swing2LimitDegrees=45
  swing2Motion=Free
  twistLimitDegrees=45
  twistMotion=Free
 }
 {bone1Name=poncho_flap_01_r, bone2Name=poncho_flap_MSTR_r, swing1LimitDegrees=3, swing2LimitDegrees=15, twistLimitDegrees=0}
 {bone1Name=poncho_flap_02_r, bone2Name=poncho_flap_01_r, swing1LimitDegrees=0, swing2LimitDegrees=15, twistLimitDegrees=0}
 {bone1Name=poncho_flap_03_r, bone2Name=poncho_flap_02_r, swing1LimitDegrees=0, swing2LimitDegrees=15, twistLimitDegrees=0}
]
